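# Physics asset for the Antique Pal Male pickaxe, defining collision and joint limits for a dangling five-bone dyn_spine chain on the two-handed handle.
# Each bone has one capsule, growing thicker along the chain to a much wider one on dyn_spine_05.
# Constraints link each segment to the previous, starting at the handle.
# The root joint has moderate, equal swing and twist limits, while later joints swing more freely but twist less.
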PhysicsAsset /BRCosmetics/Weapons/FORT_Melee/Pickaxe_Antique_Pal_Male/Meshes/Antique_Pal_Male_Axe_Physics
bodies {
 Melee_TwoHanded_Handle [{shapeType=Capsule, length=17.4703, radius=7.9457, center=<-0.3724 -0.1684 -5.0003>, rotation=<rot 0.9028 -90 0.1461>}]
 dyn_spine_01 [{
   shapeType=Capsule
   length=1.9824
   radius=1.0685
   shapeName=dyn_spine_01_capsule
   center=<0.082 -0.102 0.0053>
   rotation=<rot 0 43.5338 -88.9295>
  }]
 dyn_spine_02 [{
   shapeType=Capsule
   length=8.0032
   radius=1.9872
   shapeName=dyn_spine_02_capsule
   center=<0.7959 0.2862 0.0475>
   rotation=<rot 0.2807 -27.6906 -88.4064>
  }]
 dyn_spine_03 [{
   shapeType=Capsule
   length=10.5292
   radius=2.6714
   shapeName=dyn_spine_03_capsule
   center=<1.6782 -1.7858 -0.0061>
   rotation=<rot 0.0539 -117.8412 -89.6936>
  }]
 dyn_spine_04 [{
   shapeType=Capsule
   length=12.4438
   radius=3.2429
   shapeName=dyn_spine_04_capsule
   center=<1.8465 2.467 -0.0385>
   rotation=<rot 0.1535 -10.5978 -89.1291>
  }]
 dyn_spine_05 [{
   shapeType=Capsule
   length=13.1182
   radius=12.716
   shapeName=dyn_spine_05_capsule
   center=<-2.7287 0.0063 12.1166>
   rotation=<rot 0.8984 -99.9561 -84.6536>
  }]
}
constraints [
 {bone1Name=dyn_spine_01, bone2Name=Melee_TwoHanded_Handle, swing1LimitDegrees=20, swing2LimitDegrees=20, twistLimitDegrees=20}
 {bone1Name=dyn_spine_02, bone2Name=dyn_spine_01, swing1LimitDegrees=50, swing2LimitDegrees=50, twistLimitDegrees=10}
 {bone1Name=dyn_spine_03, bone2Name=dyn_spine_02, swing1LimitDegrees=50, swing2LimitDegrees=50, twistLimitDegrees=10}
 {bone1Name=dyn_spine_04, bone2Name=dyn_spine_03, swing1LimitDegrees=50, swing2LimitDegrees=50, twistLimitDegrees=10}
 {bone1Name=dyn_spine_05, bone2Name=dyn_spine_04, swing1LimitDegrees=50, swing2LimitDegrees=50, twistLimitDegrees=10}
]
